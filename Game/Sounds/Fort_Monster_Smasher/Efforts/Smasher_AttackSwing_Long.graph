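# Sound cue for the Smasher monster's long attack swing effort, layering close vocals with a distant re-amped version.
# The close layer mixes modulated random "Sweet" waves with louder random yells shaped by pitch and volume envelopes.
# The distant layer pitch-varies re-amped waves under its own attenuation and distant weapon-fire sound class.
# A quality switch repeats this in its second slot with fewer variations and leaves the last two empty.
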SoundCue /Game/Sounds/Fort_Monster_Smasher/Efforts/Smasher_AttackSwing_Long
root=QualityLevel_0
nodes [
 {id=Attenuation_2, children [Mixer_2], properties {attenuationSettings=/Game/Sounds/Attenuation/Fort_Monster_Smasher_Att}}
 {
  id=Attenuation_3
  children [Modulator_3]
  properties {
   attenuationSettings=/Game/Sounds/Attenuation/Fort_Monster_SmasherDistant_Att
   attenuationOverrides {
    bAttenuate=false
    bAttenuateWithLPF=true
    distanceAlgorithm=Logarithmic
    falloffDistance=8000
    lPFRadiusMax=5000
    lPFRadiusMin=2000
    attenuationShapeExtents {x=5000}
   }
  }
 }
 {id=Attenuation_4, children [Mixer_4], properties {attenuationSettings=&Fort_Monster_Smasher_Att}}
 {
  id=Attenuation_5
  children [Modulator_5]
  properties {
   attenuationSettings=&Fort_Monster_SmasherDistant_Att
   attenuationOverrides {
    bAttenuate=false
    bAttenuateWithLPF=true
    distanceAlgorithm=Logarithmic
    falloffDistance=8000
    lPFRadiusMax=5000
    lPFRadiusMin=2000
    attenuationShapeExtents {x=5000}
   }
  }
 }
 {
  id=Enveloper_1
  children [Random_4]
  properties {
   pitchCurve {editorCurveData {keys [<at 0.0022 0.8126>,<at 0.2281 1.0705>,<at 0.725 1.0954>,<at 1.0894 0.9634>]}}
   volumeCurve {editorCurveData {keys [<at 0 1>,<at 0.6046 0.983>,<at 1.4564 -0.0071>]}}
  }
 }
 {
  id=Enveloper_2
  children [Random_7]
  properties {
   pitchCurve {editorCurveData {keys [<at 0.0022 0.8126>,<at 0.2281 1.0705>,<at 0.725 1.0954>,<at 1.0894 0.9634>]}}
   volumeCurve {editorCurveData {keys [<at 0 1>,<at 0.6046 0.983>,<at 1.4564 -0.0071>]}}
  }
 }
 {id=Mixer_2, children [Modulator_2,Enveloper_1], properties {inputVolume [0.2,1.4]}}
 {id=Mixer_3, children [SoundClass_3,SoundClass_2], properties {inputVolume [1,1.5]}}
 {id=Mixer_4, children [Modulator_4,Enveloper_2], properties {inputVolume [0.2,1.4]}}
 {id=Mixer_5, children [SoundClass_5,SoundClass_4], properties {inputVolume [1,1.5]}}
 {id=Modulator_2, children [Random_3]}
 {id=Modulator_3, children [Random_5], properties {pitchMax=1.1, pitchMin=0.9}}
 {id=Modulator_4, children [Random_6]}
 {id=Modulator_5, children [Random_8], properties {pitchMax=1.1, pitchMin=0.9}}
 {id=QualityLevel_0, children [Mixer_3,Mixer_5,null,null]}
 {id=Random_3, children [WavePlayer_14, WavePlayer_15, WavePlayer_16, WavePlayer_17], properties {weights [1,1,1,1]}}
 {
  id=Random_4
  children [WavePlayer_18, WavePlayer_19, WavePlayer_20, WavePlayer_22, WavePlayer_23, WavePlayer_21]
  properties {weights [1,1,1,1,1,1]}
 }
 {id=Random_5, children [WavePlayer_24, WavePlayer_25, WavePlayer_26, WavePlayer_27], properties {weights [1,1,1,1]}}
 {id=Random_6, children [WavePlayer_28, WavePlayer_29, WavePlayer_30, WavePlayer_31], properties {weights [1,1,1,1]}}
 {id=Random_7, children [WavePlayer_32, WavePlayer_33, WavePlayer_37, WavePlayer_35], properties {weights [1,1,1,1]}}
 {id=Random_8, children [WavePlayer_38, WavePlayer_39, WavePlayer_41], properties {weights [1,1,1]}}
 {
  id=SoundClass_2
  children [Attenuation_3]
  properties {soundClassOverride=&Fort_WeaponRanged_Fire_Distant}
 }
 {id=SoundClass_3, children [Attenuation_2], properties {soundClassOverride=/Game/Sounds/FortSoundClassesAndModes/Fort_Creatures_Foley}}
 {
  id=SoundClass_4
  children [Attenuation_5]
  properties {soundClassOverride=&Fort_WeaponRanged_Fire_Distant}
 }
 {id=SoundClass_5, children [Attenuation_4], properties {soundClassOverride=&Fort_Creatures_Foley}}
 {id=WavePlayer_14, properties {soundWaveAssetPtr=&/Smasher_Sweet_02}}
 {id=WavePlayer_15, properties {soundWaveAssetPtr=&/Smasher_Sweet_03}}
 {id=WavePlayer_16, properties {soundWaveAssetPtr=&/Smasher_Sweet_04}}
 {id=WavePlayer_17, properties {soundWaveAssetPtr=&/Smasher_Sweet_09}}
 {id=WavePlayer_18, properties {soundWaveAssetPtr=/Game/Sounds/Fort_Monster_Smasher/ScaryEfforts/smasher_yells_render_023}}
 {id=WavePlayer_19, properties {soundWaveAssetPtr=/Game/Sounds/Fort_Monster_Smasher/ScaryEfforts/smasher_yells_render_024}}
 {id=WavePlayer_20, properties {soundWaveAssetPtr=/Game/Sounds/Fort_Monster_Smasher/ScaryEfforts/smasher_yells_render_025}}
 {id=WavePlayer_21, properties {soundWaveAssetPtr=/Game/Sounds/Fort_Monster_Smasher/ScaryEfforts/smasher_yells_render_026}}
 {id=WavePlayer_22, properties {soundWaveAssetPtr=/Game/Sounds/Fort_Monster_Smasher/ScaryEfforts/smasher_yells_render_030}}
 {id=WavePlayer_23, properties {soundWaveAssetPtr=/Game/Sounds/Fort_Monster_Smasher/ScaryEfforts/smasher_yells_render_031}}
 {id=WavePlayer_24, properties {soundWaveAssetPtr=/Game/Sounds/Fort_Monster_Smasher/ReAmped/Smasher_AttackSwing_Long_01}}
 {id=WavePlayer_25, properties {soundWaveAssetPtr=/Game/Sounds/Fort_Monster_Smasher/ReAmped/Smasher_AttackSwing_Long_02}}
 {id=WavePlayer_26, properties {soundWaveAssetPtr=/Game/Sounds/Fort_Monster_Smasher/ReAmped/Smasher_AttackSwing_Long_03}}
 {id=WavePlayer_27, properties {soundWaveAssetPtr=/Game/Sounds/Fort_Monster_Smasher/ReAmped/Smasher_AttackSwing_Long_04}}
 {id=WavePlayer_28, properties {soundWaveAssetPtr=&Smasher_Sweet_02}}
 {id=WavePlayer_29, properties {soundWaveAssetPtr=&Smasher_Sweet_03}}
 {id=WavePlayer_30, properties {soundWaveAssetPtr=&Smasher_Sweet_04}}
 {id=WavePlayer_31, properties {soundWaveAssetPtr=&Smasher_Sweet_09}}
 {id=WavePlayer_32, properties {soundWaveAssetPtr=&smasher_yells_render_023}}
 {id=WavePlayer_33, properties {soundWaveAssetPtr=&smasher_yells_render_024}}
 {id=WavePlayer_35, properties {soundWaveAssetPtr=&smasher_yells_render_026}}
 {id=WavePlayer_37, properties {soundWaveAssetPtr=&smasher_yells_render_031}}
 {id=WavePlayer_38, properties {soundWaveAssetPtr=&Smasher_AttackSwing_Long_01}}
 {id=WavePlayer_39, properties {soundWaveAssetPtr=&Smasher_AttackSwing_Long_02}}
 {id=WavePlayer_41, properties {soundWaveAssetPtr=&Smasher_AttackSwing_Long_04}}
]
properties {
 bOverrideConcurrency=true
 maxDistance=7000
 soundClassObject=&Fort_Creatures_Foley
 attenuationOverrides {distanceAlgorithm=NaturalSound, falloffDistance=5000, attenuationShapeExtents {x=1000}}
 concurrencyOverrides {resolutionRule=PreventNew}
}
