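# Sound cue for the Smasher monster's attack windup vocal effort.
# A near layer mixes random windup sounds with enveloped random yells under Smasher attenuation.
# It is mixed with a pitch-modulated ReAmped distant layer using low-pass distance attenuation, each layer having its own sound class.
# A quality switch picks between the full tree and a copy with fewer wave variations, leaving its last two slots empty.
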SoundCue /Game/Sounds/Fort_Monster_Smasher/Efforts/Smasher_AttackWindup
root=QualityLevel_0
nodes [
 {id=Attenuation_2, children [Mixer_2], properties {attenuationSettings=/Game/Sounds/Attenuation/Fort_Monster_Smasher_Att}}
 {
  id=Attenuation_3
  children [Modulator_5]
  properties {
   attenuationSettings=/Game/Sounds/Attenuation/Fort_Monster_SmasherDistant_Att
   attenuationOverrides {
    bAttenuate=false
    bAttenuateWithLPF=true
    distanceAlgorithm=Logarithmic
    falloffDistance=8000
    lPFRadiusMax=5000
    lPFRadiusMin=2000
    attenuationShapeExtents {x=5000}
   }
  }
 }
 {id=Attenuation_4, children [Mixer_4], properties {attenuationSettings=&Fort_Monster_Smasher_Att}}
 {
  id=Attenuation_5
  children [Modulator_8]
  properties {
   attenuationSettings=&Fort_Monster_SmasherDistant_Att
   attenuationOverrides {
    bAttenuate=false
    bAttenuateWithLPF=true
    distanceAlgorithm=Logarithmic
    falloffDistance=8000
    lPFRadiusMax=5000
    lPFRadiusMin=2000
    attenuationShapeExtents {x=5000}
   }
  }
 }
 {id=Enveloper_1, children [Random_4]}
 {
  id=Enveloper_2
  children [Random_7]
  properties {pitchCurve {editorCurveData {keys [{time=0}]}}, volumeCurve {editorCurveData {keys [<at 0 1>]}}}
 }
 {id=Mixer_2, children [Modulator_3,Modulator_4], properties {inputVolume [1,0.8]}}
 {id=Mixer_3, children [SoundClass_3,SoundClass_2], properties {inputVolume [1,1.5]}}
 {id=Mixer_4, children [Modulator_6,Modulator_7], properties {inputVolume [1,0.8]}}
 {id=Mixer_5, children [SoundClass_5,SoundClass_4], properties {inputVolume [1,1.5]}}
 {id=Modulator_3, children [Random_3]}
 {id=Modulator_4, children [Enveloper_1]}
 {id=Modulator_5, children [Random_5], properties {pitchMax=1.1, pitchMin=0.9}}
 {id=Modulator_6, children [Random_6]}
 {id=Modulator_7, children [Enveloper_2]}
 {id=Modulator_8, children [Random_8], properties {pitchMax=1.1, pitchMin=0.9}}
 {id=QualityLevel_0, children [Mixer_3,Mixer_5,null,null]}
 {id=Random_3, children [WavePlayer_17, WavePlayer_18, WavePlayer_19, WavePlayer_20], properties {weights [1,1,1,1]}}
 {
  id=Random_4
  children [WavePlayer_21, WavePlayer_25, WavePlayer_22, WavePlayer_24, WavePlayer_23, WavePlayer_28, WavePlayer_27, WavePlayer_26]
  properties {weights [1,1,1,1,1,1,1,1]}
 }
 {id=Random_5, children [WavePlayer_29, WavePlayer_30, WavePlayer_31, WavePlayer_32, WavePlayer_33], properties {weights [1,1,1,1,1]}}
 {id=Random_6, children [WavePlayer_35, WavePlayer_36, WavePlayer_37], properties {weights [1,1,1]}}
 {id=Random_7, children [WavePlayer_38, WavePlayer_42, WavePlayer_40, WavePlayer_45, WavePlayer_44], properties {weights [1,1,1,1,1]}}
 {id=Random_8, children [WavePlayer_46, WavePlayer_47, WavePlayer_49], properties {weights [1,1,1]}}
 {
  id=SoundClass_2
  children [Attenuation_3]
  properties {soundClassOverride=&Fort_WeaponRanged_Fire_Distant}
 }
 {id=SoundClass_3, children [Attenuation_2], properties {soundClassOverride=/Game/Sounds/FortSoundClassesAndModes/Fort_Enemy_Damaged_PSM}}
 {
  id=SoundClass_4
  children [Attenuation_5]
  properties {soundClassOverride=&Fort_WeaponRanged_Fire_Distant}
 }
 {id=SoundClass_5, children [Attenuation_4], properties {soundClassOverride=&Fort_Enemy_Damaged_PSM}}
 {id=WavePlayer_17, properties {soundWaveAssetPtr=&/Smasher_AttackWindup01}}
 {id=WavePlayer_18, properties {soundWaveAssetPtr=&/Smasher_AttackWindup02}}
 {id=WavePlayer_19, properties {soundWaveAssetPtr=&/Smasher_AttackWindup03}}
 {id=WavePlayer_20, properties {soundWaveAssetPtr=&/Smasher_AttackWindup04}}
 {id=WavePlayer_21, properties {soundWaveAssetPtr=&/smasher_yells_render_011}}
 {id=WavePlayer_22, properties {soundWaveAssetPtr=&/smasher_yells_render_012}}
 {id=WavePlayer_23, properties {soundWaveAssetPtr=&/smasher_yells_render_014}}
 {id=WavePlayer_24, properties {soundWaveAssetPtr=&/smasher_yells_render_021}}
 {id=WavePlayer_25, properties {soundWaveAssetPtr=&/smasher_yells_render_022}}
 {id=WavePlayer_26, properties {soundWaveAssetPtr=&/smasher_yells_render_023}}
 {id=WavePlayer_27, properties {soundWaveAssetPtr=&/smasher_yells_render_025}}
 {id=WavePlayer_28, properties {soundWaveAssetPtr=&/smasher_yells_render_031}}
 {id=WavePlayer_29, properties {soundWaveAssetPtr=/Game/Sounds/Fort_Monster_Smasher/ReAmped/Smasher_AttackWindup_01}}
 {id=WavePlayer_30, properties {soundWaveAssetPtr=/Game/Sounds/Fort_Monster_Smasher/ReAmped/Smasher_AttackWindup_02}}
 {id=WavePlayer_31, properties {soundWaveAssetPtr=/Game/Sounds/Fort_Monster_Smasher/ReAmped/Smasher_AttackWindup_03}}
 {id=WavePlayer_32, properties {soundWaveAssetPtr=/Game/Sounds/Fort_Monster_Smasher/ReAmped/Smasher_AttackWindup_04}}
 {id=WavePlayer_33, properties {soundWaveAssetPtr=/Game/Sounds/Fort_Monster_Smasher/ReAmped/Smasher_AttackWindup_05}}
 {id=WavePlayer_35, properties {soundWaveAssetPtr=&Smasher_AttackWindup02}}
 {id=WavePlayer_36, properties {soundWaveAssetPtr=&Smasher_AttackWindup03}}
 {id=WavePlayer_37, properties {soundWaveAssetPtr=&Smasher_AttackWindup04}}
 {id=WavePlayer_38, properties {soundWaveAssetPtr=&smasher_yells_render_011}}
 {id=WavePlayer_40, properties {soundWaveAssetPtr=&smasher_yells_render_014}}
 {id=WavePlayer_42, properties {soundWaveAssetPtr=&smasher_yells_render_022}}
 {id=WavePlayer_44, properties {soundWaveAssetPtr=&smasher_yells_render_025}}
 {id=WavePlayer_45, properties {soundWaveAssetPtr=&smasher_yells_render_031}}
 {id=WavePlayer_46, properties {soundWaveAssetPtr=&Smasher_AttackWindup_01}}
 {id=WavePlayer_47, properties {soundWaveAssetPtr=&Smasher_AttackWindup_02}}
 {id=WavePlayer_49, properties {soundWaveAssetPtr=&Smasher_AttackWindup_04}}
]
properties {
 bOverrideConcurrency=true
 maxDistance=7000
 volumeMultiplier=0.9
 attenuationOverrides {distanceAlgorithm=NaturalSound, falloffDistance=5000, attenuationShapeExtents {x=1000}}
 concurrencyOverrides {resolutionRule=PreventNew}
}
