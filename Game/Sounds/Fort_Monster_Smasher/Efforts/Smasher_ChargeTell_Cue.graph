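# Sound cue for the Smasher monster's charge-tell vocal, layering close and distant variations with concurrency set to prevent new instances.
# The first quality branch mixes random yells with a sweetener (sometimes pitch and volume enveloped), random distant ChargeTell waves and a quest scream.
# The second branch drops the scream and uses fewer variations; two quality slots are empty.
# Each layer gets random modulation, with close and distant attenuation and sound class overrides.
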SoundCue /Game/Sounds/Fort_Monster_Smasher/Efforts/Smasher_ChargeTell_Cue
root=QualityLevel_0
nodes [
 {id=Attenuation_0, children [Modulator_0], properties {attenuationSettings=/Game/Sounds/Attenuation/Fort_Monster_Smasher_Att}}
 {
  id=Attenuation_2
  children [Modulator_5]
  properties {
   attenuationSettings=/Game/Sounds/Attenuation/Fort_Monster_SmasherDistant_Att
   attenuationOverrides {
    bAttenuate=false
    bAttenuateWithLPF=true
    distanceAlgorithm=Logarithmic
    falloffDistance=8000
    lPFRadiusMax=5000
    lPFRadiusMin=2000
    attenuationShapeExtents {x=5000}
   }
  }
 }
 {id=Attenuation_3, children [Mixer_2], properties {attenuationSettings=&Fort_Monster_Smasher_Att}}
 {
  id=Attenuation_4
  children [Modulator_8]
  properties {
   attenuationSettings=&Fort_Monster_SmasherDistant_Att
   attenuationOverrides {
    bAttenuate=false
    bAttenuateWithLPF=true
    distanceAlgorithm=Logarithmic
    falloffDistance=8000
    lPFRadiusMax=5000
    lPFRadiusMin=2000
    attenuationShapeExtents {x=5000}
   }
  }
 }
 {id=Attenuation_5, children [Mixer_4], properties {attenuationSettings=&Fort_Monster_Smasher_Att}}
 {
  id=Enveloper_1
  children [WavePlayer_11]
  properties {pitchCurve {editorCurveData {keys [{time=0}, {time=0.5}, <at 1 1.2>]}}, volumeCurve {editorCurveData {keys [<at 0 0.8>]}}}
 }
 {
  id=Enveloper_2
  children [WavePlayer_21]
  properties {pitchCurve {editorCurveData {keys [{time=0}, {time=0.5}, <at 1 1.2>]}}, volumeCurve {editorCurveData {keys [<at 0 0.8>]}}}
 }
 {id=Mixer_2, children [Modulator_3,Modulator_4], properties {inputVolume [0.5,0.2]}}
 {id=Mixer_3, children [SoundClass_3,SoundClass_2,SoundClass_0], properties {inputVolume [0.85,1.5,1]}}
 {id=Mixer_4, children [Modulator_6,Modulator_7], properties {inputVolume [1.5,0.2]}}
 {id=Mixer_5, children [SoundClass_5,SoundClass_4], properties {inputVolume [1,1.5]}}
 {id=Modulator_0, children [WavePlayer_5], properties {pitchMax=1.1, pitchMin=0.99, volumeMax=1.15, volumeMin=0.97}}
 {id=Modulator_3, children [Random_5], properties {pitchMax=1.2, pitchMin=1}}
 {id=Modulator_4, children [Random_3], properties {pitchMax=1}}
 {id=Modulator_5, children [Random_6], properties {pitchMax=1.1, pitchMin=0.9}}
 {id=Modulator_6, children [Random_8], properties {pitchMax=1.2, pitchMin=1}}
 {id=Modulator_7, children [Random_7], properties {pitchMax=1}}
 {id=Modulator_8, children [Random_9], properties {pitchMax=1.1, pitchMin=0.9}}
 {id=QualityLevel_0, children [Mixer_3,Mixer_5,null,null]}
 {id=Random_3, children [WavePlayer_10, Enveloper_1], properties {weights [1,1]}}
 {id=Random_5, children [WavePlayer_15, WavePlayer_14, WavePlayer_13, WavePlayer_12], properties {weights [1,1,1,1]}}
 {id=Random_6, children [WavePlayer_16, WavePlayer_17, WavePlayer_18, WavePlayer_19], properties {weights [1,1,1,1]}}
 {id=Random_7, children [WavePlayer_20, Enveloper_2], properties {weights [1,1]}}
 {id=Random_8, children [WavePlayer_24, WavePlayer_23, WavePlayer_22], properties {weights [1,1,1]}}
 {id=Random_9, children [WavePlayer_26, WavePlayer_32, WavePlayer_33], properties {weights [1,1,1]}}
 {
  id=SoundClass_0
  children [Attenuation_0]
  properties {soundClassOverride=/Game/Sounds/FortSoundClassesAndModes/Fort_Creatures_Large_Critical_PSM}
 }
 {
  id=SoundClass_2
  children [Attenuation_2]
  properties {soundClassOverride=&Fort_WeaponRanged_Fire_Distant}
 }
 {
  id=SoundClass_3
  children [Attenuation_3]
  properties {soundClassOverride=&Fort_Creatures_Large_Critical_PSM}
 }
 {
  id=SoundClass_4
  children [Attenuation_4]
  properties {soundClassOverride=&Fort_WeaponRanged_Fire_Distant}
 }
 {
  id=SoundClass_5
  children [Attenuation_5]
  properties {soundClassOverride=&Fort_Creatures_Large_Critical_PSM}
 }
 {id=WavePlayer_10, properties {soundWaveAssetPtr=&/Smasher_Sweet_04}}
 {id=WavePlayer_11, properties {soundWaveAssetPtr=&Smasher_Sweet_04}}
 {id=WavePlayer_12, properties {soundWaveAssetPtr=&/smasher_yells_render_011}}
 {id=WavePlayer_13, properties {soundWaveAssetPtr=&/smasher_yells_render_062}}
 {id=WavePlayer_14, properties {soundWaveAssetPtr=&/smasher_yells_render_060}}
 {id=WavePlayer_15, properties {soundWaveAssetPtr=&/smasher_yells_render_059}}
 {id=WavePlayer_16, properties {soundWaveAssetPtr=/Game/Sounds/Fort_Monster_Smasher/ReAmped/Smasher_ChargeTell_01}}
 {id=WavePlayer_17, properties {soundWaveAssetPtr=/Game/Sounds/Fort_Monster_Smasher/ReAmped/Smasher_ChargeTell_02}}
 {id=WavePlayer_18, properties {soundWaveAssetPtr=/Game/Sounds/Fort_Monster_Smasher/ReAmped/Smasher_ChargeTell_03}}
 {id=WavePlayer_19, properties {soundWaveAssetPtr=/Game/Sounds/Fort_Monster_Smasher/ReAmped/Smasher_ChargeTell_04}}
 {id=WavePlayer_20, properties {soundWaveAssetPtr=&Smasher_Sweet_04}}
 {id=WavePlayer_21, properties {soundWaveAssetPtr=&Smasher_Sweet_04}}
 {id=WavePlayer_22, properties {soundWaveAssetPtr=&smasher_yells_render_011}}
 {id=WavePlayer_23, properties {soundWaveAssetPtr=&smasher_yells_render_062}}
 {id=WavePlayer_24, properties {soundWaveAssetPtr=&smasher_yells_render_060}}
 {id=WavePlayer_26, properties {soundWaveAssetPtr=&Smasher_ChargeTell_01}}
 {id=WavePlayer_32, properties {soundWaveAssetPtr=&Smasher_ChargeTell_03}}
 {id=WavePlayer_33, properties {soundWaveAssetPtr=&Smasher_ChargeTell_04}}
 {id=WavePlayer_5, properties {soundWaveAssetPtr=&/Quest_Plankerton_38_SMASHER_Scream_010}}
]
properties {
 bOverrideConcurrency=true
 maxDistance=7000
 volumeMultiplier=0.8
 attenuationOverrides {bAttenuateWithLPF=true, distanceAlgorithm=NaturalSound, falloffDistance=8000, attenuationShapeExtents {x=2000}}
 concurrencyOverrides {resolutionRule=PreventNew}
}
