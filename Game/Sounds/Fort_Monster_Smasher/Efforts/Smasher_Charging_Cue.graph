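# Sound cue for the Smasher monster's charging vocal effort, with a quality switch between two near-identical branches, the second using fewer wave variations, and two empty slots.
# Each branch mixes a close creature layer of randomized screams, pitched-down "Sweet" waves and yells under a rising pitch envelope with a louder distant layer of re-amped charging waves.
# Each layer has its own sound class and attenuation, the distant one using distance low-pass filtering without volume falloff.
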SoundCue /Game/Sounds/Fort_Monster_Smasher/Efforts/Smasher_Charging_Cue
root=QualityLevel_0
nodes [
 {
  id=Attenuation_2
  children [Modulator_5]
  properties {
   attenuationSettings=/Game/Sounds/Attenuation/Fort_Monster_SmasherDistant_Att
   attenuationOverrides {
    bAttenuate=false
    bAttenuateWithLPF=true
    distanceAlgorithm=Logarithmic
    falloffDistance=8000
    lPFRadiusMax=5000
    lPFRadiusMin=2000
    attenuationShapeExtents {x=5000}
   }
  }
 }
 {id=Attenuation_3, children [Mixer_2], properties {attenuationSettings=/Game/Sounds/Attenuation/Fort_Monster_Smasher_Att}}
 {
  id=Attenuation_4
  children [Modulator_8]
  properties {
   attenuationSettings=&Fort_Monster_SmasherDistant_Att
   attenuationOverrides {
    bAttenuate=false
    bAttenuateWithLPF=true
    distanceAlgorithm=Logarithmic
    falloffDistance=8000
    lPFRadiusMax=5000
    lPFRadiusMin=2000
    attenuationShapeExtents {x=5000}
   }
  }
 }
 {id=Attenuation_5, children [Mixer_4], properties {attenuationSettings=&Fort_Monster_Smasher_Att}}
 {
  id=Enveloper_1
  children [Random_6]
  properties {pitchCurve {editorCurveData {keys [<at 0.2208 0.5525>,<at 3.1783 0.9705>]}}, volumeCurve {editorCurveData {keys [<at 0 1.0119>]}}}
 }
 {
  id=Enveloper_2
  children [Random_10]
  properties {pitchCurve {editorCurveData {keys [<at 0.2208 0.5525>,<at 3.1783 0.9705>]}}, volumeCurve {editorCurveData {keys [<at 0 1.0119>]}}}
 }
 {id=Mixer_2, children [Modulator_3,Modulator_4,Enveloper_1], properties {inputVolume [1,0.5,1]}}
 {id=Mixer_3, children [SoundClass_3,SoundClass_2], properties {inputVolume [1,1.5]}}
 {id=Mixer_4, children [Modulator_6,Modulator_7,Enveloper_2], properties {inputVolume [1,0.5,1]}}
 {id=Mixer_5, children [SoundClass_5,SoundClass_4], properties {inputVolume [1,1.5]}}
 {id=Modulator_3, children [Random_4]}
 {id=Modulator_4, children [Random_5], properties {pitchMax=0.8, pitchMin=0.7}}
 {id=Modulator_5, children [Random_7], properties {pitchMax=1.1, pitchMin=0.9}}
 {id=Modulator_6, children [Random_8]}
 {id=Modulator_7, children [Random_9], properties {pitchMax=0.8, pitchMin=0.7}}
 {id=Modulator_8, children [Random_11], properties {pitchMax=1.1, pitchMin=0.9}}
 {id=QualityLevel_0, children [Mixer_3,Mixer_5,null,null]}
 {id=Random_10, children [WavePlayer_31, WavePlayer_33, WavePlayer_34], properties {weights [1,1,1]}}
 {id=Random_11, children [WavePlayer_35, WavePlayer_36, WavePlayer_38], properties {weights [1,1,1]}}
 {id=Random_4, children [WavePlayer_13, WavePlayer_14, WavePlayer_15], properties {weights [1,1,1]}}
 {id=Random_5, children [WavePlayer_17, WavePlayer_16], properties {weights [1,1]}}
 {id=Random_6, children [WavePlayer_18, WavePlayer_20, WavePlayer_19, WavePlayer_21], properties {weights [1,1,1,1]}}
 {id=Random_7, children [WavePlayer_22, WavePlayer_23, WavePlayer_24, WavePlayer_25], properties {weights [1,1,1,1]}}
 {id=Random_8, children [WavePlayer_26, WavePlayer_27, WavePlayer_28], properties {weights [1,1,1]}}
 {id=Random_9, children [WavePlayer_30, WavePlayer_29], properties {weights [1,1]}}
 {
  id=SoundClass_2
  children [Attenuation_2]
  properties {soundClassOverride=&Fort_WeaponRanged_Fire_Distant}
 }
 {
  id=SoundClass_3
  children [Attenuation_3]
  properties {soundClassOverride=/Game/Sounds/FortSoundClassesAndModes/Fort_Creatures_Large_Critical_PSM}
 }
 {
  id=SoundClass_4
  children [Attenuation_4]
  properties {soundClassOverride=&Fort_WeaponRanged_Fire_Distant}
 }
 {
  id=SoundClass_5
  children [Attenuation_5]
  properties {soundClassOverride=&Fort_Creatures_Large_Critical_PSM}
 }
 {id=WavePlayer_13, properties {soundWaveAssetPtr=&/Smasher_DragAttackScream_01}}
 {id=WavePlayer_14, properties {soundWaveAssetPtr=&/Smasher_DragAttackScream_02}}
 {id=WavePlayer_15, properties {soundWaveAssetPtr=&/Smasher_DragAttackScream_03}}
 {id=WavePlayer_16, properties {soundWaveAssetPtr=&/Smasher_Sweet_09}}
 {id=WavePlayer_17, properties {soundWaveAssetPtr=&/Smasher_Sweet_10}}
 {id=WavePlayer_18, properties {soundWaveAssetPtr=&/smasher_yells_render_01}}
 {id=WavePlayer_19, properties {soundWaveAssetPtr=&/smasher_yells_render_03}}
 {id=WavePlayer_20, properties {soundWaveAssetPtr=&/smasher_yells_render_07}}
 {id=WavePlayer_21, properties {soundWaveAssetPtr=&/smasher_yells_render_08}}
 {id=WavePlayer_22, properties {soundWaveAssetPtr=/Game/Sounds/Fort_Monster_Smasher/ReAmped/Smasher_Charging_02}}
 {id=WavePlayer_23, properties {soundWaveAssetPtr=/Game/Sounds/Fort_Monster_Smasher/ReAmped/Smasher_Charging_01}}
 {id=WavePlayer_24, properties {soundWaveAssetPtr=/Game/Sounds/Fort_Monster_Smasher/ReAmped/Smasher_Charging_03}}
 {id=WavePlayer_25, properties {soundWaveAssetPtr=/Game/Sounds/Fort_Monster_Smasher/ReAmped/Smasher_Charging_04}}
 {id=WavePlayer_26, properties {soundWaveAssetPtr=&Smasher_DragAttackScream_01}}
 {id=WavePlayer_27, properties {soundWaveAssetPtr=&Smasher_DragAttackScream_02}}
 {id=WavePlayer_28, properties {soundWaveAssetPtr=&Smasher_DragAttackScream_03}}
 {id=WavePlayer_29, properties {soundWaveAssetPtr=&Smasher_Sweet_09}}
 {id=WavePlayer_30, properties {soundWaveAssetPtr=&Smasher_Sweet_10}}
 {id=WavePlayer_31, properties {soundWaveAssetPtr=&smasher_yells_render_01}}
 {id=WavePlayer_33, properties {soundWaveAssetPtr=&smasher_yells_render_07}}
 {id=WavePlayer_34, properties {soundWaveAssetPtr=&smasher_yells_render_08}}
 {id=WavePlayer_35, properties {soundWaveAssetPtr=&Smasher_Charging_02}}
 {id=WavePlayer_36, properties {soundWaveAssetPtr=&Smasher_Charging_01}}
 {id=WavePlayer_38, properties {soundWaveAssetPtr=&Smasher_Charging_04}}
]
properties {
 bDebug=true
 bOverrideConcurrency=true
 maxDistance=7000
 volumeMultiplier=0.8
 attenuationOverrides {distanceAlgorithm=NaturalSound, falloffDistance=5000, attenuationShapeExtents {x=1000}}
 concurrencyOverrides {resolutionRule=PreventNew}
}
