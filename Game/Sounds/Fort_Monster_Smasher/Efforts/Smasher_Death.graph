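# Sound cue for the Smasher monster's death, routed to the enemy-damaged sound class.
# Quality switch feeds two populated branches, the second with fewer yell variations.
# Each mixes a near layer (random pitch-modulated yell, death sweetener, delayed pitched-up dematerialize) with a louder layer of random re-amped death sounds under distant attenuation.
# Near and distant layers carry separate attenuation nodes, and a concurrency override prevents new instances once the limit is reached.
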SoundCue /Game/Sounds/Fort_Monster_Smasher/Efforts/Smasher_Death
root=QualityLevel_0
nodes [
 {id=Attenuation_2, children [Mixer_2], properties {attenuationSettings=/Game/Sounds/Attenuation/Fort_Monster_Smasher_Att}}
 {
  id=Attenuation_3
  children [Modulator_5]
  properties {
   attenuationSettings=/Game/Sounds/Attenuation/Fort_Monster_SmasherDistant_Att
   attenuationOverrides {
    bAttenuate=false
    bAttenuateWithLPF=true
    distanceAlgorithm=Logarithmic
    falloffDistance=8000
    lPFRadiusMax=5000
    lPFRadiusMin=2000
    attenuationShapeExtents {x=5000}
   }
  }
 }
 {id=Attenuation_4, children [Mixer_4], properties {attenuationSettings=&Fort_Monster_Smasher_Att}}
 {
  id=Attenuation_5
  children [Modulator_8]
  properties {
   attenuationSettings=&Fort_Monster_SmasherDistant_Att
   attenuationOverrides {
    bAttenuate=false
    bAttenuateWithLPF=true
    distanceAlgorithm=Logarithmic
    falloffDistance=8000
    lPFRadiusMax=5000
    lPFRadiusMin=2000
    attenuationShapeExtents {x=5000}
   }
  }
 }
 {id=Delay_1, children [WavePlayer_11], properties {delayMax=0.5, delayMin=0.5}}
 {id=Delay_2, children [WavePlayer_22], properties {delayMax=0.5, delayMin=0.5}}
 {id=Mixer_2, children [Modulator_4, WavePlayer_12, Modulator_3], properties {inputVolume [1,0.3,0.5]}}
 {id=Mixer_3, children [Attenuation_2, Attenuation_3], properties {inputVolume [1,1.5]}}
 {id=Mixer_4, children [Modulator_7, WavePlayer_23, Modulator_6], properties {inputVolume [1,0.3,0.5]}}
 {id=Mixer_5, children [Attenuation_4, Attenuation_5], properties {inputVolume [1,1.5]}}
 {id=Modulator_3, children [Delay_1], properties {pitchMax=1.3, pitchMin=1.1}}
 {id=Modulator_4, children [Random_2], properties {pitchMax=1.2}}
 {id=Modulator_5, children [Random_3]}
 {id=Modulator_6, children [Delay_2], properties {pitchMax=1.3, pitchMin=1.1}}
 {id=Modulator_7, children [Random_4], properties {pitchMax=1.2}}
 {id=Modulator_8, children [Random_5]}
 {id=QualityLevel_0, children [Mixer_3,Mixer_5,null,null]}
 {
  id=Random_2
  children [WavePlayer_17, WavePlayer_16, WavePlayer_15, WavePlayer_14, WavePlayer_18, WavePlayer_13]
  properties {weights [1,1,1,1,1,1]}
 }
 {id=Random_3, children [WavePlayer_19, WavePlayer_20, WavePlayer_21], properties {weights [1,1,1]}}
 {id=Random_4, children [WavePlayer_28, WavePlayer_27, WavePlayer_25, WavePlayer_29], properties {weights [1,1,1,1]}}
 {id=Random_5, children [WavePlayer_30, WavePlayer_31, WavePlayer_32], properties {weights [1,1,1]}}
 {id=WavePlayer_11, properties {soundWaveAssetPtr=/Game/Sounds/Fort_Monster_Smasher/SFX/smasher_demat_01}}
 {id=WavePlayer_12, properties {soundWaveAssetPtr=/Game/Sounds/Fort_UI_Sounds/Gameplay/death_sweeten_04}}
 {id=WavePlayer_13, properties {soundWaveAssetPtr=&/smasher_yells_render_057}}
 {id=WavePlayer_14, properties {soundWaveAssetPtr=&/smasher_yells_render_058}}
 {id=WavePlayer_15, properties {soundWaveAssetPtr=&/smasher_yells_render_059}}
 {id=WavePlayer_16, properties {soundWaveAssetPtr=&/smasher_yells_render_042}}
 {id=WavePlayer_17, properties {soundWaveAssetPtr=&/smasher_yells_render_023}}
 {id=WavePlayer_18, properties {soundWaveAssetPtr=&/smasher_yells_render_013}}
 {id=WavePlayer_19, properties {soundWaveAssetPtr=/Game/Sounds/Fort_Monster_Smasher/ReAmped/Smasher_Death_01}}
 {id=WavePlayer_20, properties {soundWaveAssetPtr=/Game/Sounds/Fort_Monster_Smasher/ReAmped/Smasher_Death_02}}
 {id=WavePlayer_21, properties {soundWaveAssetPtr=/Game/Sounds/Fort_Monster_Smasher/ReAmped/Smasher_Death_03}}
 {id=WavePlayer_22, properties {soundWaveAssetPtr=&smasher_demat_01}}
 {id=WavePlayer_23, properties {soundWaveAssetPtr=&death_sweeten_04}}
 {id=WavePlayer_25, properties {soundWaveAssetPtr=&smasher_yells_render_058}}
 {id=WavePlayer_27, properties {soundWaveAssetPtr=&smasher_yells_render_042}}
 {id=WavePlayer_28, properties {soundWaveAssetPtr=&smasher_yells_render_023}}
 {id=WavePlayer_29, properties {soundWaveAssetPtr=&smasher_yells_render_013}}
 {id=WavePlayer_30, properties {soundWaveAssetPtr=&Smasher_Death_01}}
 {id=WavePlayer_31, properties {soundWaveAssetPtr=&Smasher_Death_02}}
 {id=WavePlayer_32, properties {soundWaveAssetPtr=&Smasher_Death_03}}
]
properties {
 bOverrideConcurrency=true
 maxDistance=7000
 soundClassObject=/Game/Sounds/FortSoundClassesAndModes/Fort_Enemy_Damaged_PSM
 volumeMultiplier=1
 attenuationOverrides {distanceAlgorithm=NaturalSound, falloffDistance=5000, attenuationShapeExtents {x=1000}}
 concurrencyOverrides {resolutionRule=PreventNew}
}
