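# Sound cue for the Smasher monster's tired breaths as it decelerates, layering a close and a distant version.
# A mixer feeds both branches, each picking randomly from three waves through a modulator, attenuation and sound class override.
# The close layer is pitched up in the non-critical enemy class; the louder distant layer uses ReAmped waves, distant attenuation and borrows the ranged-weapon distant-fire class.
# Concurrency is overridden to prevent new instances at the limit.
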SoundCue /Game/Sounds/Fort_Monster_Smasher/Efforts/Smasher_Decelerate_TiredBreaths_Cue
root=Mixer_46
nodes [
 {id=Attenuation_92, children [Modulator_3], properties {attenuationSettings=/Game/Sounds/Attenuation/Fort_Monster_Smasher_Att}}
 {
  id=Attenuation_93
  children [Modulator_47]
  properties {
   attenuationSettings=/Game/Sounds/Attenuation/Fort_Monster_SmasherDistant_Att
   attenuationOverrides {
    bAttenuate=false
    bAttenuateWithLPF=true
    distanceAlgorithm=Logarithmic
    falloffDistance=8000
    lPFRadiusMax=5000
    lPFRadiusMin=2000
    attenuationShapeExtents {x=5000}
   }
  }
 }
 {id=Mixer_46, children [SoundClass_93, SoundClass_92], properties {inputVolume [1,1.5]}}
 {id=Modulator_3, children [Random_1], properties {pitchMax=1.2, pitchMin=1.1}}
 {id=Modulator_47, children [Random_27]}
 {id=Random_1, children [WavePlayer_4,WavePlayer_5,WavePlayer_6], properties {weights [1,1,1]}}
 {id=Random_27, children [WavePlayer_120, WavePlayer_121, WavePlayer_122], properties {weights [1,1,1]}}
 {
  id=SoundClass_92
  children [Attenuation_93]
  properties {soundClassOverride=&Fort_WeaponRanged_Fire_Distant}
 }
 {id=SoundClass_93, children [Attenuation_92], properties {soundClassOverride=/Game/Sounds/FortSoundClassesAndModes/Fort_Enemy_NonCritical}}
 {id=WavePlayer_120, properties {soundWaveAssetPtr=/Game/Sounds/Fort_Monster_Smasher/ReAmped/Smasher_Decelerate_TiredBreaths_01}}
 {id=WavePlayer_121, properties {soundWaveAssetPtr=/Game/Sounds/Fort_Monster_Smasher/ReAmped/Smasher_Decelerate_TiredBreaths_02}}
 {id=WavePlayer_122, properties {soundWaveAssetPtr=/Game/Sounds/Fort_Monster_Smasher/ReAmped/Smasher_Decelerate_TiredBreaths_03}}
 {id=WavePlayer_4, properties {soundWaveAssetPtr=&/smasher_decelerate_tiredbreath_01}}
 {id=WavePlayer_5, properties {soundWaveAssetPtr=&/smasher_decelerate_tiredbreath_02}}
 {id=WavePlayer_6, properties {soundWaveAssetPtr=&/smasher_decelerate_tiredbreath_03}}
]
properties {bOverrideConcurrency=true, maxDistance=7000, volumeMultiplier=0.7, concurrencyOverrides {resolutionRule=PreventNew}}
