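# Sound cue for the Smasher monster's short idle grunts, with concurrency set to prevent new instances at the limit.
# A quality switch fills two of four slots with near-identical mixes, the second having one fewer distant variation.
# Each mixes random close grunts with louder, pitch-varied ReAmped distant grunts, each layer having its own attenuation and sound class.
# The distant layer overrides attenuation to low-pass filtering without volume falloff and borrows the ranged-weapon distant-fire class.
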SoundCue /Game/Sounds/Fort_Monster_Smasher/Efforts/Smasher_IdleGrunt_Short
root=QualityLevel_0
nodes [
 {id=Attenuation_2, children [Modulator_2], properties {attenuationSettings=/Game/Sounds/Attenuation/Fort_Monster_Smasher_Att}}
 {
  id=Attenuation_3
  children [Modulator_3]
  properties {
   attenuationSettings=/Game/Sounds/Attenuation/Fort_Monster_SmasherDistant_Att
   attenuationOverrides {
    bAttenuate=false
    bAttenuateWithLPF=true
    distanceAlgorithm=Logarithmic
    falloffDistance=8000
    lPFRadiusMax=5000
    lPFRadiusMin=2000
    attenuationShapeExtents {x=5000}
   }
  }
 }
 {id=Attenuation_4, children [Modulator_4], properties {attenuationSettings=&Fort_Monster_Smasher_Att}}
 {
  id=Attenuation_5
  children [Modulator_5]
  properties {
   attenuationSettings=&Fort_Monster_SmasherDistant_Att
   attenuationOverrides {
    bAttenuate=false
    bAttenuateWithLPF=true
    distanceAlgorithm=Logarithmic
    falloffDistance=8000
    lPFRadiusMax=5000
    lPFRadiusMin=2000
    attenuationShapeExtents {x=5000}
   }
  }
 }
 {id=Mixer_1, children [SoundClass_3,SoundClass_2], properties {inputVolume [1,1.5]}}
 {id=Mixer_2, children [SoundClass_5,SoundClass_4], properties {inputVolume [1,1.5]}}
 {id=Modulator_2, children [Random_2]}
 {id=Modulator_3, children [Random_3], properties {pitchMax=1.1, pitchMin=0.9}}
 {id=Modulator_4, children [Random_4]}
 {id=Modulator_5, children [Random_5], properties {pitchMax=1.1, pitchMin=0.9}}
 {id=QualityLevel_0, children [Mixer_1,Mixer_2,null,null]}
 {id=Random_2, children [WavePlayer_11, WavePlayer_9, WavePlayer_10, WavePlayer_12], properties {weights [1,1,1,1]}}
 {id=Random_3, children [WavePlayer_13, WavePlayer_14, WavePlayer_15, WavePlayer_16, WavePlayer_17], properties {weights [1,1,1,1,1]}}
 {id=Random_4, children [WavePlayer_20, WavePlayer_18, WavePlayer_19, WavePlayer_21], properties {weights [1,1,1,1]}}
 {id=Random_5, children [WavePlayer_22, WavePlayer_23, WavePlayer_24, WavePlayer_25], properties {weights [1,1,1,1]}}
 {
  id=SoundClass_2
  children [Attenuation_3]
  properties {soundClassOverride=&Fort_WeaponRanged_Fire_Distant}
 }
 {id=SoundClass_3, children [Attenuation_2], properties {soundClassOverride=/Game/Sounds/FortSoundClassesAndModes/Fort_Enemy_NonCritical}}
 {
  id=SoundClass_4
  children [Attenuation_5]
  properties {soundClassOverride=&Fort_WeaponRanged_Fire_Distant}
 }
 {id=SoundClass_5, children [Attenuation_4], properties {soundClassOverride=&Fort_Enemy_NonCritical}}
 {id=WavePlayer_10, properties {soundWaveAssetPtr=&/Smasher_IdleGrunt_Short03}}
 {id=WavePlayer_11, properties {soundWaveAssetPtr=&/Smasher_IdleGrunt_Short06}}
 {id=WavePlayer_12, properties {soundWaveAssetPtr=&/Smasher_IdleGrunt_Short02}}
 {id=WavePlayer_13, properties {soundWaveAssetPtr=/Game/Sounds/Fort_Monster_Smasher/ReAmped/Smasher_IdleGrunt_Short_01}}
 {id=WavePlayer_14, properties {soundWaveAssetPtr=/Game/Sounds/Fort_Monster_Smasher/ReAmped/Smasher_IdleGrunt_Short_02}}
 {id=WavePlayer_15, properties {soundWaveAssetPtr=/Game/Sounds/Fort_Monster_Smasher/ReAmped/Smasher_IdleGrunt_Short_03}}
 {id=WavePlayer_16, properties {soundWaveAssetPtr=/Game/Sounds/Fort_Monster_Smasher/ReAmped/Smasher_IdleGrunt_Short_04}}
 {id=WavePlayer_17, properties {soundWaveAssetPtr=/Game/Sounds/Fort_Monster_Smasher/ReAmped/Smasher_IdleGrunt_Short_05}}
 {id=WavePlayer_18, properties {soundWaveAssetPtr=&/Smasher_IdleGrunt_Short05}}
 {id=WavePlayer_19, properties {soundWaveAssetPtr=&Smasher_IdleGrunt_Short03}}
 {id=WavePlayer_20, properties {soundWaveAssetPtr=&Smasher_IdleGrunt_Short06}}
 {id=WavePlayer_21, properties {soundWaveAssetPtr=&Smasher_IdleGrunt_Short02}}
 {id=WavePlayer_22, properties {soundWaveAssetPtr=&Smasher_IdleGrunt_Short_01}}
 {id=WavePlayer_23, properties {soundWaveAssetPtr=&Smasher_IdleGrunt_Short_02}}
 {id=WavePlayer_24, properties {soundWaveAssetPtr=&Smasher_IdleGrunt_Short_03}}
 {id=WavePlayer_25, properties {soundWaveAssetPtr=&Smasher_IdleGrunt_Short_04}}
 {id=WavePlayer_9, properties {soundWaveAssetPtr=&Smasher_IdleGrunt_Short05}}
]
properties {
 bOverrideConcurrency=true
 maxDistance=7000
 attenuationOverrides {distanceAlgorithm=NaturalSound, falloffDistance=5000, attenuationShapeExtents {x=1000}}
 concurrencyOverrides {resolutionRule=PreventNew}
}
