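# Smasher enemy vocal cue for noticing a player, split by a quality-level node into two populated branches.
# Each branch mixes a near layer in the non-critical enemy sound class with a re-amped distant layer in the distant weapon-fire class, each with its own attenuation.
# Randomized waves pass through pitch/volume modulators; the first branch adds yells and a scream, while the second uses sweeteners and fewer variants.
# Concurrency override prevents new instances at the limit.
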SoundCue /Game/Sounds/Fort_Monster_Smasher/Efforts/Smasher_NoticePlayer
root=QualityLevel_0
nodes [
 {id=Attenuation_2, children [Mixer_2], properties {attenuationSettings=/Game/Sounds/Attenuation/Fort_Monster_Smasher_Att}}
 {
  id=Attenuation_3
  children [Modulator_5]
  properties {
   attenuationSettings=/Game/Sounds/Attenuation/Fort_Monster_SmasherDistant_Att
   attenuationOverrides {
    bAttenuate=false
    bAttenuateWithLPF=true
    distanceAlgorithm=Logarithmic
    falloffDistance=8000
    lPFRadiusMax=5000
    lPFRadiusMin=2000
    attenuationShapeExtents {x=5000}
   }
  }
 }
 {id=Attenuation_4, children [Mixer_4], properties {attenuationSettings=&Fort_Monster_Smasher_Att}}
 {
  id=Attenuation_5
  children [Modulator_8]
  properties {
   attenuationSettings=&Fort_Monster_SmasherDistant_Att
   attenuationOverrides {
    bAttenuate=false
    bAttenuateWithLPF=true
    distanceAlgorithm=Logarithmic
    falloffDistance=8000
    lPFRadiusMax=5000
    lPFRadiusMin=2000
    attenuationShapeExtents {x=5000}
   }
  }
 }
 {id=Mixer_2, children [Modulator_3,Modulator_4,Modulator_0], properties {inputVolume [1,0.5,1]}}
 {id=Mixer_3, children [SoundClass_3,SoundClass_2], properties {inputVolume [1,1.5]}}
 {id=Mixer_4, children [Modulator_6,Modulator_7], properties {inputVolume [1,0.4]}}
 {id=Mixer_5, children [SoundClass_5,SoundClass_4], properties {inputVolume [1,1.5]}}
 {id=Modulator_0, children [WavePlayer_5], properties {pitchMax=1.1, pitchMin=0.99, volumeMax=1.15, volumeMin=0.97}}
 {id=Modulator_3, children [Random_3]}
 {id=Modulator_4, children [Random_0], properties {pitchMax=1.3}}
 {id=Modulator_5, children [Random_5], properties {pitchMax=1.1, pitchMin=0.9}}
 {id=Modulator_6, children [Random_6]}
 {id=Modulator_7, children [Random_7], properties {pitchMax=1.3}}
 {id=Modulator_8, children [Random_8], properties {pitchMax=1.1, pitchMin=0.9}}
 {id=QualityLevel_0, children [Mixer_3,Mixer_5,null,null]}
 {id=Random_0, children [WavePlayer_1,WavePlayer_0], properties {weights [1,1]}}
 {id=Random_3, children [WavePlayer_12, WavePlayer_13, WavePlayer_14, WavePlayer_15], properties {weights [1,1,1,1]}}
 {id=Random_5, children [WavePlayer_19, WavePlayer_20, WavePlayer_21, WavePlayer_22, WavePlayer_23], properties {weights [1,1,1,1,1]}}
 {id=Random_6, children [WavePlayer_24, WavePlayer_26, WavePlayer_27], properties {weights [1,1,1]}}
 {id=Random_7, children [WavePlayer_28, WavePlayer_29, WavePlayer_30], properties {weights [1,1,1]}}
 {id=Random_8, children [WavePlayer_32, WavePlayer_34, WavePlayer_35], properties {weights [1,1,1]}}
 {
  id=SoundClass_2
  children [Attenuation_3]
  properties {soundClassOverride=&Fort_WeaponRanged_Fire_Distant}
 }
 {id=SoundClass_3, children [Attenuation_2], properties {soundClassOverride=/Game/Sounds/FortSoundClassesAndModes/Fort_Enemy_NonCritical}}
 {
  id=SoundClass_4
  children [Attenuation_5]
  properties {soundClassOverride=&Fort_WeaponRanged_Fire_Distant}
 }
 {id=SoundClass_5, children [Attenuation_4], properties {soundClassOverride=&Fort_Enemy_NonCritical}}
 {id=WavePlayer_0, properties {soundWaveAssetPtr=&/smasher_yells_render_014}}
 {id=WavePlayer_1, properties {soundWaveAssetPtr=&/smasher_yells_render_061}}
 {id=WavePlayer_12, properties {soundWaveAssetPtr=&/Smasher_NoticePlayer01}}
 {id=WavePlayer_13, properties {soundWaveAssetPtr=&/Smasher_NoticePlayer02}}
 {id=WavePlayer_14, properties {soundWaveAssetPtr=&/Smasher_NoticePlayer05}}
 {id=WavePlayer_15, properties {soundWaveAssetPtr=&/Smasher_NoticePlayer10}}
 {id=WavePlayer_19, properties {soundWaveAssetPtr=/Game/Sounds/Fort_Monster_Smasher/ReAmped/Smasher_NoticePlayer_01}}
 {id=WavePlayer_20, properties {soundWaveAssetPtr=/Game/Sounds/Fort_Monster_Smasher/ReAmped/Smasher_NoticePlayer_02}}
 {id=WavePlayer_21, properties {soundWaveAssetPtr=/Game/Sounds/Fort_Monster_Smasher/ReAmped/Smasher_NoticePlayer_03}}
 {id=WavePlayer_22, properties {soundWaveAssetPtr=/Game/Sounds/Fort_Monster_Smasher/ReAmped/Smasher_NoticePlayer_04}}
 {id=WavePlayer_23, properties {soundWaveAssetPtr=/Game/Sounds/Fort_Monster_Smasher/ReAmped/Smasher_NoticePlayer_05}}
 {id=WavePlayer_24, properties {soundWaveAssetPtr=&Smasher_NoticePlayer01}}
 {id=WavePlayer_26, properties {soundWaveAssetPtr=&Smasher_NoticePlayer05}}
 {id=WavePlayer_27, properties {soundWaveAssetPtr=&Smasher_NoticePlayer10}}
 {id=WavePlayer_28, properties {soundWaveAssetPtr=&/Smasher_Sweet_01}}
 {id=WavePlayer_29, properties {soundWaveAssetPtr=&/Smasher_Sweet_02}}
 {id=WavePlayer_30, properties {soundWaveAssetPtr=&/Smasher_Sweet_04}}
 {id=WavePlayer_32, properties {soundWaveAssetPtr=&Smasher_NoticePlayer_02}}
 {id=WavePlayer_34, properties {soundWaveAssetPtr=&Smasher_NoticePlayer_04}}
 {id=WavePlayer_35, properties {soundWaveAssetPtr=&Smasher_NoticePlayer_05}}
 {id=WavePlayer_5, properties {soundWaveAssetPtr=&/Quest_Plankerton_38_SMASHER_Scream_010}}
]
properties {
 bOverrideConcurrency=true
 maxDistance=7000
 volumeMultiplier=1
 attenuationOverrides {distanceAlgorithm=NaturalSound, falloffDistance=5000, attenuationShapeExtents {x=1000}}
 concurrencyOverrides {resolutionRule=PreventNew}
}
