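# Sound cue for the Smasher monster's short pain vocal, limited to one instance per owner, with new plays blocked.
# A quality switch fills two of four slots with identically structured branches, the second using fewer wave variations.
# Each mixes a near enemy-damaged layer of randomized pain, "Sweet" and yell waves with a louder-weighted re-amped layer using distant attenuation and the distant ranged-weapon-fire sound class.
# Modulators vary playback per layer.
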SoundCue /Game/Sounds/Fort_Monster_Smasher/Efforts/Smasher_Pain_Short
root=QualityLevel_0
nodes [
 {id=Attenuation_2, children [Mixer_2], properties {attenuationSettings=/Game/Sounds/Attenuation/Fort_Monster_Smasher_Att}}
 {
  id=Attenuation_3
  children [Modulator_7]
  properties {
   attenuationSettings=/Game/Sounds/Attenuation/Fort_Monster_SmasherDistant_Att
   attenuationOverrides {
    bAttenuate=false
    bAttenuateWithLPF=true
    distanceAlgorithm=Logarithmic
    falloffDistance=8000
    lPFRadiusMax=5000
    lPFRadiusMin=2000
    attenuationShapeExtents {x=5000}
   }
  }
 }
 {id=Attenuation_4, children [Mixer_4], properties {attenuationSettings=&Fort_Monster_Smasher_Att}}
 {
  id=Attenuation_5
  children [Modulator_11]
  properties {
   attenuationSettings=&Fort_Monster_SmasherDistant_Att
   attenuationOverrides {
    bAttenuate=false
    bAttenuateWithLPF=true
    distanceAlgorithm=Logarithmic
    falloffDistance=8000
    lPFRadiusMax=5000
    lPFRadiusMin=2000
    attenuationShapeExtents {x=5000}
   }
  }
 }
 {id=Mixer_2, children [Modulator_4,Modulator_5,Modulator_6], properties {inputVolume [1,0.4,1]}}
 {id=Mixer_3, children [SoundClass_3,SoundClass_2], properties {inputVolume [1,1.5]}}
 {id=Mixer_4, children [Modulator_8,Modulator_9,Modulator_10], properties {inputVolume [1,0.4,1]}}
 {id=Mixer_5, children [SoundClass_5,SoundClass_4], properties {inputVolume [1,1.5]}}
 {id=Modulator_10, children [Random_10]}
 {id=Modulator_11, children [Random_11], properties {pitchMax=1.1, pitchMin=0.9}}
 {id=Modulator_4, children [Random_4]}
 {id=Modulator_5, children [Random_5], properties {pitchMax=1.3, pitchMin=0.9}}
 {id=Modulator_6, children [Random_6]}
 {id=Modulator_7, children [Random_7], properties {pitchMax=1.1, pitchMin=0.9}}
 {id=Modulator_8, children [Random_8]}
 {id=Modulator_9, children [Random_9], properties {pitchMax=1.3, pitchMin=0.9}}
 {id=QualityLevel_0, children [Mixer_3,Mixer_5,null,null]}
 {id=Random_10, children [WavePlayer_53, WavePlayer_54, WavePlayer_55, WavePlayer_51, WavePlayer_52], properties {weights [1,1,1,1,1]}}
 {id=Random_11, children [WavePlayer_57, WavePlayer_59, WavePlayer_61, WavePlayer_62], properties {weights [1,1,1,1]}}
 {id=Random_4, children [WavePlayer_21, WavePlayer_22, WavePlayer_23, WavePlayer_24, WavePlayer_25], properties {weights [1,1,1,1,1]}}
 {id=Random_5, children [WavePlayer_26, WavePlayer_27, WavePlayer_28], properties {weights [1,1,1]}}
 {
  id=Random_6
  children [WavePlayer_32, WavePlayer_33, WavePlayer_35, WavePlayer_34, WavePlayer_30, WavePlayer_29, WavePlayer_31]
  properties {weights [1,1,1,1,1,1,1]}
 }
 {
  id=Random_7
  children [WavePlayer_36, WavePlayer_37, WavePlayer_38, WavePlayer_39, WavePlayer_40, WavePlayer_41]
  properties {weights [1,1,1,1,1,1]}
 }
 {id=Random_8, children [WavePlayer_43, WavePlayer_44, WavePlayer_45, WavePlayer_46], properties {weights [1,1,1,1]}}
 {id=Random_9, children [WavePlayer_47, WavePlayer_48, WavePlayer_49], properties {weights [1,1,1]}}
 {
  id=SoundClass_2
  children [Attenuation_3]
  properties {soundClassOverride=&Fort_WeaponRanged_Fire_Distant}
 }
 {id=SoundClass_3, children [Attenuation_2], properties {soundClassOverride=/Game/Sounds/FortSoundClassesAndModes/Fort_Enemy_Damaged_PSM}}
 {
  id=SoundClass_4
  children [Attenuation_5]
  properties {soundClassOverride=&Fort_WeaponRanged_Fire_Distant}
 }
 {id=SoundClass_5, children [Attenuation_4], properties {soundClassOverride=&Fort_Enemy_Damaged_PSM}}
 {id=WavePlayer_21, properties {soundWaveAssetPtr=&/Smasher_Pain_Short01}}
 {id=WavePlayer_22, properties {soundWaveAssetPtr=&/Smasher_Pain_Short02}}
 {id=WavePlayer_23, properties {soundWaveAssetPtr=&/Smasher_Pain_Short03}}
 {id=WavePlayer_24, properties {soundWaveAssetPtr=&/Smasher_Pain_Short04}}
 {id=WavePlayer_25, properties {soundWaveAssetPtr=&/Smasher_Pain_Short06}}
 {id=WavePlayer_26, properties {soundWaveAssetPtr=&/Smasher_Sweet_12}}
 {id=WavePlayer_27, properties {soundWaveAssetPtr=&/Smasher_Sweet_13}}
 {id=WavePlayer_28, properties {soundWaveAssetPtr=&/Smasher_Sweet_14}}
 {id=WavePlayer_29, properties {soundWaveAssetPtr=&/smasher_yells_render_014}}
 {id=WavePlayer_30, properties {soundWaveAssetPtr=&/smasher_yells_render_015}}
 {id=WavePlayer_31, properties {soundWaveAssetPtr=&/smasher_yells_render_016}}
 {id=WavePlayer_32, properties {soundWaveAssetPtr=&/smasher_yells_render_017}}
 {id=WavePlayer_33, properties {soundWaveAssetPtr=&/smasher_yells_render_030}}
 {id=WavePlayer_34, properties {soundWaveAssetPtr=&/smasher_yells_render_032}}
 {id=WavePlayer_35, properties {soundWaveAssetPtr=&/smasher_yells_render_036}}
 {id=WavePlayer_36, properties {soundWaveAssetPtr=/Game/Sounds/Fort_Monster_Smasher/ReAmped/Smasher_Pain_Short_01}}
 {id=WavePlayer_37, properties {soundWaveAssetPtr=/Game/Sounds/Fort_Monster_Smasher/ReAmped/Smasher_Pain_Short_02}}
 {id=WavePlayer_38, properties {soundWaveAssetPtr=/Game/Sounds/Fort_Monster_Smasher/ReAmped/Smasher_Pain_Short_03}}
 {id=WavePlayer_39, properties {soundWaveAssetPtr=/Game/Sounds/Fort_Monster_Smasher/ReAmped/Smasher_Pain_Short_04}}
 {id=WavePlayer_40, properties {soundWaveAssetPtr=/Game/Sounds/Fort_Monster_Smasher/ReAmped/Smasher_Pain_Short_05}}
 {id=WavePlayer_41, properties {soundWaveAssetPtr=/Game/Sounds/Fort_Monster_Smasher/ReAmped/Smasher_Pain_Short_06}}
 {id=WavePlayer_43, properties {soundWaveAssetPtr=&Smasher_Pain_Short02}}
 {id=WavePlayer_44, properties {soundWaveAssetPtr=&Smasher_Pain_Short03}}
 {id=WavePlayer_45, properties {soundWaveAssetPtr=&Smasher_Pain_Short04}}
 {id=WavePlayer_46, properties {soundWaveAssetPtr=&Smasher_Pain_Short06}}
 {id=WavePlayer_47, properties {soundWaveAssetPtr=&Smasher_Sweet_12}}
 {id=WavePlayer_48, properties {soundWaveAssetPtr=&Smasher_Sweet_13}}
 {id=WavePlayer_49, properties {soundWaveAssetPtr=&Smasher_Sweet_14}}
 {id=WavePlayer_51, properties {soundWaveAssetPtr=&smasher_yells_render_015}}
 {id=WavePlayer_52, properties {soundWaveAssetPtr=&smasher_yells_render_016}}
 {id=WavePlayer_53, properties {soundWaveAssetPtr=&smasher_yells_render_017}}
 {id=WavePlayer_54, properties {soundWaveAssetPtr=&smasher_yells_render_030}}
 {id=WavePlayer_55, properties {soundWaveAssetPtr=&smasher_yells_render_032}}
 {id=WavePlayer_57, properties {soundWaveAssetPtr=&Smasher_Pain_Short_01}}
 {id=WavePlayer_59, properties {soundWaveAssetPtr=&Smasher_Pain_Short_03}}
 {id=WavePlayer_61, properties {soundWaveAssetPtr=&Smasher_Pain_Short_05}}
 {id=WavePlayer_62, properties {soundWaveAssetPtr=&Smasher_Pain_Short_06}}
]
properties {
 bOverrideConcurrency=true
 maxDistance=7000
 volumeMultiplier=0.8889
 concurrencyOverrides {bLimitToOwner=true, maxCount=1, resolutionRule=PreventNew}
}
