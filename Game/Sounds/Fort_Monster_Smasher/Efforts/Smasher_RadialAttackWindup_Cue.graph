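# Sound cue for the Smasher monster's radial attack windup effort.
# Each branch mixes a close layer of random windup and sweetener waves, classed as enemy attack, with a re-amped distant layer using its own attenuation and the distant ranged-fire sound class.
# A quality-level root feeds two such branches, the second with fewer wave variations, and leaves two slots empty.
# Overridden concurrency prevents new instances at the limit.
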SoundCue /Game/Sounds/Fort_Monster_Smasher/Efforts/Smasher_RadialAttackWindup_Cue
root=QualityLevel_0
nodes [
 {id=Attenuation_2, children [Mixer_2], properties {attenuationSettings=/Game/Sounds/Attenuation/Fort_Monster_Smasher_Att}}
 {
  id=Attenuation_3
  children [Modulator_3]
  properties {
   attenuationSettings=/Game/Sounds/Attenuation/Fort_Monster_SmasherDistant_Att
   attenuationOverrides {
    bAttenuate=false
    bAttenuateWithLPF=true
    distanceAlgorithm=Logarithmic
    falloffDistance=8000
    lPFRadiusMax=5000
    lPFRadiusMin=2000
    attenuationShapeExtents {x=5000}
   }
  }
 }
 {id=Attenuation_4, children [Mixer_4], properties {attenuationSettings=&Fort_Monster_Smasher_Att}}
 {
  id=Attenuation_5
  children [Modulator_5]
  properties {
   attenuationSettings=&Fort_Monster_SmasherDistant_Att
   attenuationOverrides {
    bAttenuate=false
    bAttenuateWithLPF=true
    distanceAlgorithm=Logarithmic
    falloffDistance=8000
    lPFRadiusMax=5000
    lPFRadiusMin=2000
    attenuationShapeExtents {x=5000}
   }
  }
 }
 {id=Mixer_2, children [Modulator_2,Random_4], properties {inputVolume [1.1,0.7]}}
 {id=Mixer_3, children [SoundClass_3,SoundClass_2], properties {inputVolume [1,1.5]}}
 {id=Mixer_4, children [Modulator_4,Random_7], properties {inputVolume [1.1,0.7]}}
 {id=Mixer_5, children [SoundClass_5,SoundClass_4], properties {inputVolume [1,1.5]}}
 {id=Modulator_2, children [Random_3]}
 {id=Modulator_3, children [Random_5]}
 {id=Modulator_4, children [Random_6]}
 {id=Modulator_5, children [Random_8]}
 {id=QualityLevel_0, children [Mixer_3,Mixer_5,null,null]}
 {id=Random_3, children [WavePlayer_10, WavePlayer_11, WavePlayer_12, WavePlayer_13], properties {weights [1,1,1,1]}}
 {id=Random_4, children [WavePlayer_14, WavePlayer_15], properties {weights [1,1]}}
 {id=Random_5, children [WavePlayer_16, WavePlayer_17, WavePlayer_18, WavePlayer_19], properties {weights [1,1,1,1]}}
 {id=Random_6, children [WavePlayer_21, WavePlayer_22, WavePlayer_23], properties {weights [1,1,1]}}
 {id=Random_7, children [WavePlayer_24, WavePlayer_25], properties {weights [1,1]}}
 {id=Random_8, children [WavePlayer_26, WavePlayer_27, WavePlayer_28], properties {weights [1,1,1]}}
 {
  id=SoundClass_2
  children [Attenuation_3]
  properties {soundClassOverride=&Fort_WeaponRanged_Fire_Distant}
 }
 {id=SoundClass_3, children [Attenuation_2], properties {soundClassOverride=/Game/Sounds/FortSoundClassesAndModes/Fort_Enemy_Attack}}
 {
  id=SoundClass_4
  children [Attenuation_5]
  properties {soundClassOverride=&Fort_WeaponRanged_Fire_Distant}
 }
 {id=SoundClass_5, children [Attenuation_4], properties {soundClassOverride=&Fort_Enemy_Attack}}
 {id=WavePlayer_10, properties {soundWaveAssetPtr=&/Smasher_AttackWindup01}}
 {id=WavePlayer_11, properties {soundWaveAssetPtr=&/Smasher_AttackWindup02}}
 {id=WavePlayer_12, properties {soundWaveAssetPtr=&/Smasher_AttackWindup03}}
 {id=WavePlayer_13, properties {soundWaveAssetPtr=&/Smasher_AttackWindup04}}
 {id=WavePlayer_14, properties {soundWaveAssetPtr=&/Smasher_Sweet_01}}
 {id=WavePlayer_15, properties {soundWaveAssetPtr=&/Smasher_Sweet_03}}
 {id=WavePlayer_16, properties {soundWaveAssetPtr=/Game/Sounds/Fort_Monster_Smasher/ReAmped/Smasher_RadialAttackWindup_01}}
 {id=WavePlayer_17, properties {soundWaveAssetPtr=/Game/Sounds/Fort_Monster_Smasher/ReAmped/Smasher_RadialAttackWindup_02}}
 {id=WavePlayer_18, properties {soundWaveAssetPtr=/Game/Sounds/Fort_Monster_Smasher/ReAmped/Smasher_RadialAttackWindup_03}}
 {id=WavePlayer_19, properties {soundWaveAssetPtr=/Game/Sounds/Fort_Monster_Smasher/ReAmped/Smasher_RadialAttackWindup_04}}
 {id=WavePlayer_21, properties {soundWaveAssetPtr=&Smasher_AttackWindup02}}
 {id=WavePlayer_22, properties {soundWaveAssetPtr=&Smasher_AttackWindup03}}
 {id=WavePlayer_23, properties {soundWaveAssetPtr=&Smasher_AttackWindup04}}
 {id=WavePlayer_24, properties {soundWaveAssetPtr=&Smasher_Sweet_01}}
 {id=WavePlayer_25, properties {soundWaveAssetPtr=&Smasher_Sweet_03}}
 {id=WavePlayer_26, properties {soundWaveAssetPtr=&Smasher_RadialAttackWindup_01}}
 {id=WavePlayer_27, properties {soundWaveAssetPtr=&Smasher_RadialAttackWindup_02}}
 {id=WavePlayer_28, properties {soundWaveAssetPtr=&Smasher_RadialAttackWindup_03}}
]
properties {
 bOverrideConcurrency=true
 maxDistance=7000
 volumeMultiplier=0.8889
 attenuationOverrides {distanceAlgorithm=NaturalSound, falloffDistance=5000, attenuationShapeExtents {x=1000}}
 concurrencyOverrides {resolutionRule=PreventNew}
}
